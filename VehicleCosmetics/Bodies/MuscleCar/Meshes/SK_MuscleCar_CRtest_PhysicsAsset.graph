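# Physics asset for the CRtest muscle car skeletal mesh, giving its chassis, glass, steering, axle and wheel bones simple collision bodies.
# Each bone listed has a single capsule, including the wheel discs and glass panels.
# Constraints form a tree: steering wheel, glass and axle pivots join the body, which joins the frame, while each wheel disc reaches the frame through a steering or knuckle bone.
# All joints share identical swing and twist limits, with no per-part tuning.
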PhysicsAsset /VehicleCosmetics/Bodies/MuscleCar/Meshes/SK_MuscleCar_CRtest_PhysicsAsset
bodies {
 axle_pivot_bk_l [{shapeType=Capsule, length=44.7612, radius=28.0821, center=<0.5467 -38.1794 6.1931>, rotation=<rot 0 -0.4721 -75.9975>}]
 axle_pivot_bk_r [{shapeType=Capsule, length=44.7612, radius=28.0821, center=<0.5467 38.1794 6.1931>, rotation=<rot 0 -179.5279 -75.9975>}]
 axle_pivot_fr_l [{shapeType=Capsule, length=37.1719, radius=29.2672, center=<-0.0365 -29.9972 6.6674>, rotation=<rot 0 0.0564 -72.4056>}]
 axle_pivot_fr_r [{shapeType=Capsule, length=36.0599, radius=29.1753, center=<-0.0266 29.0095 5.8235>, rotation=<rot 0 179.9575 -70.2393>}]
 body [{shapeType=Capsule, length=244.7864, radius=117.2039, center=<47.2807 0.0161 11.4276>, rotation=<rot 0 -89.9818 -84.8219>}]
 frame [{shapeType=Capsule, length=12.2684, radius=12.2633, center=<163.858 82.0435 18.0886>, rotation=<rot -1.833 -90 10.7458>}]
 glass_c [{shapeType=Capsule, length=70.2619, radius=32.8683, center=<-32.5962 0.0014 14.4556>, rotation=<rot 0 -0.0029 -89.9983>}]
 glass_l [{shapeType=Capsule, length=68.1845, radius=13.3947, center=<7.8553 8.5238 10.2425>, rotation=<rot 0 -93.8489 -85.5263>}]
 glass_r [{shapeType=Capsule, length=68.1845, radius=13.3947, center=<7.8554 -8.5238 10.2425>, rotation=<rot 0 -86.1512 -85.5263>}]
 steering_knuckle_bk_l [{shapeType=Capsule, length=16.6492, radius=13.0607, center=<-4.5153 -2.476 4.6917>, rotation=<rot 0 -71.0955 -49.9296>}]
 steering_knuckle_bk_r [{shapeType=Capsule, length=16.6492, radius=13.0607, center=<-4.5153 2.476 4.6917>, rotation=<rot 0 -108.9045 -49.9296>}]
 steering_wheel [{shapeType=Capsule, length=19.5252, radius=15.3045, center=<2.246 -0.3752 0.9084>, rotation=<rot 0 -179.988 -87.4799>}]
 wheel_disc_bk_l [{shapeType=Capsule, length=21.2656, radius=21.2642, center=<0 -2.192 0.0028>, rotation=<rot 0.0817 -90 25.4493>}]
 wheel_disc_bk_r [{shapeType=Capsule, length=21.2656, radius=21.2642, center=<0 2.192 0.0028>, rotation=<rot -0.0817 -90 25.4493>}]
 wheel_disc_fr_l [{shapeType=Capsule, length=21.3194, radius=21.3172, center=<0 -2.0298 0.0031>, rotation=<rot 0.0914 -90 17.0728>}]
 wheel_disc_fr_r [{shapeType=Capsule, length=21.3194, radius=21.3172, center=<0 2.0298 0.0031>, rotation=<rot -0.0913 -90 17.0729>}]
 wheel_steering_fr_l [{shapeType=Capsule, length=19.4025, radius=12.8394, center=<-6.5847 0.979 4.6187>, rotation=<rot 0 -66.4747 -55.8803>}]
 wheel_steering_fr_r [{shapeType=Capsule, length=19.4233, radius=12.7105, center=<-4.8029 3.0308 4.6406>, rotation=<rot 0 -105.8411 -55.6829>}]
}
constraints [
 {bone1Name=body, bone2Name=frame, swing1LimitDegrees=45, swing2LimitDegrees=45, twistLimitDegrees=45}
 {bone1Name=steering_wheel, bone2Name=body, swing1LimitDegrees=45, swing2LimitDegrees=45, twistLimitDegrees=45}
 {bone1Name=glass_c, bone2Name=body, swing1LimitDegrees=45, swing2LimitDegrees=45, twistLimitDegrees=45}
 {bone1Name=glass_l, bone2Name=body, swing1LimitDegrees=45, swing2LimitDegrees=45, twistLimitDegrees=45}
 {bone1Name=glass_r, bone2Name=body, swing1LimitDegrees=45, swing2LimitDegrees=45, twistLimitDegrees=45}
 {bone1Name=axle_pivot_fr_l, bone2Name=body, swing1LimitDegrees=45, swing2LimitDegrees=45, twistLimitDegrees=45}
 {bone1Name=axle_pivot_fr_r, bone2Name=body, swing1LimitDegrees=45, swing2LimitDegrees=45, twistLimitDegrees=45}
 {bone1Name=axle_pivot_bk_l, bone2Name=body, swing1LimitDegrees=45, swing2LimitDegrees=45, twistLimitDegrees=45}
 {bone1Name=axle_pivot_bk_r, bone2Name=body, swing1LimitDegrees=45, swing2LimitDegrees=45, twistLimitDegrees=45}
 {bone1Name=wheel_steering_fr_l, bone2Name=frame, swing1LimitDegrees=45, swing2LimitDegrees=45, twistLimitDegrees=45}
 {bone1Name=wheel_disc_fr_l, bone2Name=wheel_steering_fr_l, swing1LimitDegrees=45, swing2LimitDegrees=45, twistLimitDegrees=45}
 {bone1Name=wheel_steering_fr_r, bone2Name=frame, swing1LimitDegrees=45, swing2LimitDegrees=45, twistLimitDegrees=45}
 {bone1Name=wheel_disc_fr_r, bone2Name=wheel_steering_fr_r, swing1LimitDegrees=45, swing2LimitDegrees=45, twistLimitDegrees=45}
 {bone1Name=steering_knuckle_bk_l, bone2Name=frame, swing1LimitDegrees=45, swing2LimitDegrees=45, twistLimitDegrees=45}
 {bone1Name=wheel_disc_bk_l, bone2Name=steering_knuckle_bk_l, swing1LimitDegrees=45, swing2LimitDegrees=45, twistLimitDegrees=45}
 {bone1Name=steering_knuckle_bk_r, bone2Name=frame, swing1LimitDegrees=45, swing2LimitDegrees=45, twistLimitDegrees=45}
 {bone1Name=wheel_disc_bk_r, bone2Name=steering_knuckle_bk_r, swing1LimitDegrees=45, swing2LimitDegrees=45, twistLimitDegrees=45}
]
